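# Control Rig for the JoyfulGrin male character's face accessory that correctively adjusts the headband bone as the neck and head move.
# Two spherical pose readers measure neck_01 and head orientation relative to spine_05.
# Each output weights a following transform modifier that shifts and pitches HeadBand_MSTR, with the head-driven one also squashing it in Z.
# The dynamic headband chains and face bones are in the hierarchy but not directly driven here.
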ControlRigBlueprint /BRCosmetics/Characters/Player/Male/Medium/Bodies/M_MED_JoyfulGrin/Meshes/Parts/M_MED_JoyfulGrin_FaceAcc_CtrlRig
graphs [
 RigVMFunctionLibrary {links [], nodes []}
 RigVMModel {nodes [
   RigUnit_BeginExecution = &RigUnit_BeginExecution()
   n1 = &RigUnit_SphericalPoseReader(
    ActiveRegionScaleFactors=(PositiveWidth=0.5,NegativeWidth=1,PositiveHeight=3,NegativeHeight=3)
    ActiveRegionSize=0.1
    Debug=(bDrawDebug=false,DebugScale=25,DebugSegments=20,DebugThickness=0.25)
    DriverAxis=(X=0,Y=-1,Z=0)
    DriverItem=(Type=Bone,Name="neck_01")
    FalloffRegionScaleFactors=(PositiveWidth=0.5,NegativeWidth=0.5,PositiveHeight=2,NegativeHeight=2)
    FalloffSize=0.35
    FlipWidthScaling=false
    OptionalParentItem=(Type=Bone,Name="spine_05")
    RotationOffset=(X=-90,Y=0,Z=-40)
    ExecutePin=@RigUnit_BeginExecution.ExecutePin
   )
   > &RigUnit_ModifyTransforms(
    ItemToModify=((Item=(Type=Bone,Name="HeadBand_MSTR"),Transform=(Rotation=(X=0,Y=0,Z=0,W=1),Translation=(X=-1,Y=0,Z=3),Scale3D=(X=1,Y=1,Z=1))))
    ItemToModify.0.Transform.Rotation=&RigVMFunction_MathQuaternionFromEuler(Euler=(X=0,Y=-30,Z=0)).Result
    Weight=@n1.OutputParam
   )
   > n2 = &RigUnit_SphericalPoseReader(
    ActiveRegionScaleFactors=(PositiveWidth=3,NegativeWidth=3,PositiveHeight=1.5,NegativeHeight=1.5)
    ActiveRegionSize=0.1
    Debug=(bDrawDebug=false,bDraw2D=false,bDrawLocalAxes=false,DebugScale=25,DebugSegments=20,DebugThickness=0.25)
    DriverAxis=(X=1,Y=0,Z=0)
    DriverItem=(Type=Bone,Name="head")
    FalloffRegionScaleFactors=(PositiveWidth=2,NegativeWidth=2,PositiveHeight=0.75,NegativeHeight=0.75)
    FalloffSize=0.35
    OptionalParentItem=(Type=Bone,Name="spine_05")
    RotationOffset=(X=-90,Y=-90,Z=0)
   )
   > &RigUnit_ModifyTransforms(
    ItemToModify=((Item=(Type=Bone,Name="HeadBand_MSTR"),Transform=(Rotation=(X=0,Y=0,Z=0,W=1),Translation=(X=-0.5,Y=0,Z=2.5),Scale3D=(X=1,Y=1,Z=0.75))))
    ItemToModify.0.Transform.Rotation=&RigVMFunction_MathQuaternionFromEuler(Euler=(X=0,Y=-30,Z=0)).Result
    Weight=@n2.OutputParam
   )
  ]}
]
hierarchy [
 root {loc=<0 0 0>, rot=<0 0 0 1>}
 pelvis {parent=root, loc=<0 -0.7071 93.1191>, rot=<0.0127 0.707 -0.0127 -0.707>}
 spine_01 {parent=pelvis, loc=<6.7494 -0.6388 0>, rot=<0 0 0.034 -0.9994>}
 spine_02 {parent=spine_01, loc=<7.2518 -0.035 0>, rot=<0 0 0.0238 -0.9997>}
 spine_03 {parent=spine_02, loc=<6.0104 0 0>, rot=<0 0 0.0149 -0.9999>}
 spine_04 {parent=spine_03, loc=<8.1024 0 0>, rot=<0 0 -0.0132 -0.9999>}
 spine_05 {parent=spine_04, loc=<10.0146 0 0>, rot=<0 0 -0.0015 -1>}
 neck_01 {parent=spine_05, loc=<11.3138 0.1529 0>, rot=<0 0 -0.2003 -0.9797>}
 neck_02 {parent=neck_01, loc=<4.5627 0 0>, rot=<0 0 0.0504 -0.9987>}
 head {parent=neck_02, loc=<5.3914 0 0>, rot=<0 0 0.115 -0.9934>}
 HeadBand_MSTR {parent=head, loc=<6.6614 -7.6181 0>, rot=<0.4989 -0.5007 0.4952 -0.5052>}
 dyn_HeadBand_l_MSTR {parent=HeadBand_MSTR, loc=<0.2078 2.2126 -0.0392>, rot=<0.2525 -0.2583 -0.6816 -0.6363>}
 dyn_HeadBand_l_01_jnt {parent=dyn_HeadBand_l_MSTR, loc=<0 0 0>, rot=<0 0 0 1>}
 dyn_HeadBand_l_02_jnt {parent=dyn_HeadBand_l_01_jnt, loc=<7.6502 -0.064 0.0066>, rot=<0.0767 -0.0075 -0.3944 -0.9157>}
 dyn_HeadBand_l_03_jnt {parent=dyn_HeadBand_l_02_jnt, loc=<8.0122 0 0>, rot=<-0.1723 -0.12 -0.0577 -0.976>}
 dyn_HeadBand_l_04_jnt {parent=dyn_HeadBand_l_03_jnt, loc=<8.7187 0 0>, rot=<-0.2142 -0.0782 0.1099 -0.9674>}
 dyn_HeadBand_l_05_jnt {parent=dyn_HeadBand_l_04_jnt, loc=<8.8277 0 0>, rot=<-0.1397 -0.0245 0.1873 -0.972>}
 dyn_HeadBand_l_06_jnt {parent=dyn_HeadBand_l_05_jnt, loc=<9.0272 0 0>, rot=<0.1286 -0.0105 0.1778 -0.9756>}
 dyn_HeadBand_l_07_jnt {parent=dyn_HeadBand_l_06_jnt, loc=<7.463 0 0>, rot=<-0.0192 0.0233 0.0471 -0.9984>}
 dyn_HeadBand_r_MSTR {parent=HeadBand_MSTR, loc=<0.6766 -1.368 -1.1432>, rot=<0.6605 -0.6992 -0.1967 -0.1902>}
 dyn_HeadBand_r_01_jnt {parent=dyn_HeadBand_r_MSTR, loc=<0 0 0>, rot=<0 0 0 1>}
 dyn_HeadBand_r_02_jnt {parent=dyn_HeadBand_r_01_jnt, loc=<6.0567 0 0>, rot=<0.0239 0.151 -0.3178 -0.9357>}
 dyn_HeadBand_r_03_jnt {parent=dyn_HeadBand_r_02_jnt, loc=<8.0868 0 0>, rot=<0.2111 0.1222 -0.0656 -0.9676>}
 dyn_HeadBand_r_04_jnt {parent=dyn_HeadBand_r_03_jnt, loc=<11.1437 0 0>, rot=<0.2402 0.0372 0.0781 -0.9669>}
 dyn_HeadBand_r_05_jnt {parent=dyn_HeadBand_r_04_jnt, loc=<9.4524 0 0>, rot=<0.0851 -0.0425 0.2253 -0.9696>}
 dyn_HeadBand_r_06_jnt {parent=dyn_HeadBand_r_05_jnt, loc=<9.7513 0 0>, rot=<-0.0116 -0.0136 0.1525 -0.9881>}
 dyn_HeadBand_r_07_jnt {parent=dyn_HeadBand_r_06_jnt, loc=<9.8157 0 0>, rot=<-0.0276 0.0465 0.0744 -0.9958>}
 faceAttach {parent=head, loc=<2.216 -2.168 -0.208>, rot=<0.0029 -0.7071 -0.0029 -0.7071>}
 C_brow_mid {parent=faceAttach, loc=<-0.2084 14.1784 3.5887>, rot=<0 0 0 -1>}
 C_jaw {parent=faceAttach, loc=<0.0161 2.1046 -1.5567>, rot=<0 0 0 1>}
 C_lip_upper_mid {parent=faceAttach, loc=<-0.2084 15.5584 -5.6674>, rot=<0.2225 0.0013 -0.0005 -0.9749>}
 C_nose_bridge {parent=faceAttach, loc=<-0.2084 15.505 0.4146>, rot=<-0.2218 0 0 -0.9751>}
 L_brow_mid {parent=faceAttach, loc=<4.96 12.0803 4.3968>, rot=<-0.0001 -0.0042 0.2553 -0.9669>}
 L_brow_outer {parent=faceAttach, loc=<7.2868 8.6708 4.666>, rot=<-0.0287 -0.04 0.4491 -0.8921>}
 L_cheek_inner {parent=faceAttach, loc=<8.3746 7.561 -1.5983>, rot=<-0.0706 -0.2697 0.3184 -0.906>}
 L_eye {parent=faceAttach, loc=<4.5687 9.1642 2.1636>, rot=<0 0 0.9998 0.0175>}
 L_eye_lid_lower_mid {parent=L_eye, loc=<0 0 0>, rot=<0.0329 0 0 -0.9995>}
 L_eye_lid_upper_mid {parent=L_eye, loc=<0 0 0>, rot=<0.0523 0 0 -0.9986>}
 L_lip_corner {parent=faceAttach, loc=<7.187 9.932 -4.1296>, rot=<0.003 0.008 -0.0146 -0.9999>}
 L_lip_upper_outer {parent=faceAttach, loc=<5.4426 13.1462 -4.806>, rot=<0.0359 -0.0728 0.2627 -0.9615>}
 R_brow_mid {parent=faceAttach, loc=<-5.3768 12.0803 4.3968>, rot=<-0.0001 0.0042 -0.2553 -0.9669>}
 R_brow_outer {parent=faceAttach, loc=<-7.7036 8.6708 4.666>, rot=<-0.0287 0.04 -0.4491 -0.8921>}
 R_cheek_inner {parent=faceAttach, loc=<-8.7914 7.561 -1.5983>, rot=<-0.0706 0.2697 -0.3184 -0.906>}
 R_eye {parent=faceAttach, loc=<-4.9855 9.1642 2.1635>, rot=<0 0 0.9998 -0.0175>}
 R_eye_lid_lower_mid {parent=R_eye, loc=<0 0 -0.0001>, rot=<0.0329 0 0 -0.9995>}
 R_eye_lid_upper_mid {parent=R_eye, loc=<0 0 0>, rot=<0.0523 0 0 -0.9986>}
 R_lip_corner {parent=faceAttach, loc=<-7.6038 9.932 -4.1296>, rot=<0.003 -0.008 0.0146 -0.9999>}
 R_lip_upper_outer {parent=faceAttach, loc=<-5.8594 13.1462 -4.806>, rot=<0.0359 0.0728 -0.2627 -0.9615>}
 clavicle_l {parent=spine_05, loc=<4.3575 -0.1466 -2.3223>, rot=<0.0465 -0.7289 -0.009 -0.6829>}
 clavicle_r {parent=spine_05, loc=<4.3575 -0.1466 2.3223>, rot=<0.7289 0.0465 -0.6829 0.009>}
]
