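# Anim Blueprint for the JoyfulGrin male face accessory, driving its headband bones with secondary motion.
# Two parallel chains copy the parent mesh pose through a Control Rig, constrain the headband joints to head, neck, clavicles and spine, then apply offsets and RigidBody simulation.
# Skydive floating or dive mode selects the Skydive physics asset chain over the Ground one.
# Each update also sets gravity override via a deprecated helper.
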AnimBlueprint /BRCosmetics/Characters/Player/Male/Medium/Bodies/M_MED_JoyfulGrin/Meshes/Parts/M_MED_JoyfulGrin_FaceAcc_AnimBP
target_skeleton=&/M_MED_JoyfulGrin_FaceAcc_Skeleton
graphs [
 AnimGraph {nodes [
   n1 = Control Rig(Source=Copy Pose From Mesh() {bUseAttachedParent=true}) {
    alphaBoolBlend {blendOption=Linear}
    controlRigAssetReference {blueprintRigClass=&/M_MED_JoyfulGrin_FaceAcc_CtrlRig.M_MED_JoyfulGrin_FaceAcc_CtrlRig_C}
   }
   n2 = "Constraint - dyn_HeadBand_l_01_jnt (PP)"!(ComponentPose=Local To Component(LocalPose=@n1), Alpha=1, ConstraintWeights_0=0.9, ConstraintWeights_1:float=0.1) -> (Pose) {
    alphaBoolBlend {blendOption=Linear}
    boneToModify {boneName=dyn_HeadBand_l_01_jnt}
    constraintSetup [{targetBone {boneName=head}}, {targetBone {boneName=neck_01}}]
    constraintWeights [1,1]
   }
   n3 = "Constraint - dyn_HeadBand_r_01_jnt (PP)"!(ComponentPose=@n2, Alpha=1, ConstraintWeights_0=0.9, ConstraintWeights_1:float=0.1) -> (Pose) {
    alphaBoolBlend {blendOption=Linear}
    boneToModify {boneName=dyn_HeadBand_r_01_jnt}
    constraintSetup [{targetBone {boneName=head}}, {targetBone {boneName=neck_01}}]
    constraintWeights [1,1]
   }
   n4 = "Constraint - dyn_HeadBand_l_02_jnt (PP)"!(ComponentPose=@n3, Alpha=1, ConstraintWeights_0=0.1, ConstraintWeights_1:float=0.9) -> (Pose) {
    alphaBoolBlend {blendOption=Linear}
    boneToModify {boneName=dyn_HeadBand_l_02_jnt}
    constraintSetup [{targetBone {boneName=clavicle_l}}, {targetBone {boneName=neck_01}}]
    constraintWeights [1,1]
   }
   n5 = "Constraint - dyn_HeadBand_r_02_jnt (PP)"!(ComponentPose=@n4, Alpha=1, ConstraintWeights_0=0.1, ConstraintWeights_1:float=0.9) -> (Pose) {
    alphaBoolBlend {blendOption=Linear}
    boneToModify {boneName=dyn_HeadBand_r_02_jnt}
    constraintSetup [{targetBone {boneName=clavicle_r}}, {targetBone {boneName=neck_01}}]
    constraintWeights [1,1]
   }
   n6 = "Transform (Modify) Bone - Bone: dyn_HeadBand_l_02_jnt"(ComponentPose=@n5, Alpha=1, Translation="0,-2,0", Rotation="0,0,0", Scale="1,1,1") {
    rotationMode=BMM_Additive
    rotationSpace=BCS_BoneSpace
    translationMode=BMM_Additive
    translationSpace=BCS_BoneSpace
    alphaBoolBlend {blendOption=Linear}
    boneToModify {boneName=dyn_HeadBand_l_02_jnt}
    translation {y=-2}
   }
   n7 = "Transform (Modify) Bone - Bone: dyn_HeadBand_r_02_jnt"(ComponentPose=@n6, Alpha=1, Translation="0,-1,0", Rotation="0,0,60", Scale="1,1,1") {
    rotationMode=BMM_Additive
    rotationSpace=BCS_BoneSpace
    translationMode=BMM_Additive
    translationSpace=BCS_BoneSpace
    alphaBoolBlend {blendOption=Linear}
    boneToModify {boneName=dyn_HeadBand_r_02_jnt}
    rotation {roll=60}
    translation {y=-1}
   }
   n8 = "Constraint - dyn_HeadBand_l_03_jnt (PP)"!(ComponentPose=@n7, Alpha=1, ConstraintWeights_0=0.5, ConstraintWeights_1:float=0.5) -> (Pose) {
    alphaBoolBlend {blendOption=Linear}
    boneToModify {boneName=dyn_HeadBand_l_03_jnt}
    constraintSetup [{targetBone {boneName=clavicle_l}}, {targetBone {boneName=spine_05}}]
    constraintWeights [1,1]
   }
   n9 = "Constraint - dyn_HeadBand_r_03_jnt (PP)"!(ComponentPose=@n8, Alpha=1, ConstraintWeights_0=0.5, ConstraintWeights_1:float=0.5) -> (Pose) {
    alphaBoolBlend {blendOption=Linear}
    boneToModify {boneName=dyn_HeadBand_r_03_jnt}
    constraintSetup [{targetBone {boneName=clavicle_r}}, {targetBone {boneName=spine_05}}]
    constraintWeights [1,1]
   }
   n10 = "Transform (Modify) Bone - Bone: dyn_HeadBand_l_03_jnt"(ComponentPose=@n9, Alpha=1, Translation="0,2,0", Rotation="0,0,0", Scale="1,1,1") {
    rotationMode=BMM_Additive
    rotationSpace=BCS_BoneSpace
    translationMode=BMM_Additive
    translationSpace=BCS_BoneSpace
    alphaBoolBlend {blendOption=Linear}
    boneToModify {boneName=dyn_HeadBand_l_03_jnt}
    translation {y=2}
   }
   n11 = "Transform (Modify) Bone - Bone: dyn_HeadBand_r_03_jnt"(ComponentPose=@n10, Alpha=1, Translation="0,0,0", Rotation="0,0,0", Scale="1,1,1") {
    rotationMode=BMM_Additive
    rotationSpace=BCS_BoneSpace
    translationMode=BMM_Additive
    translationSpace=BCS_BoneSpace
    alphaBoolBlend {blendOption=Linear}
    boneToModify {boneName=dyn_HeadBand_r_03_jnt}
   }
   n12 = "Constraint - dyn_HeadBand_l_04_jnt (PP)"!(ComponentPose=@n11, Alpha=1, ConstraintWeights_0=0.75, ConstraintWeights_1:float=0.25) -> (Pose) {
    alphaBoolBlend {blendOption=Linear}
    boneToModify {boneName=dyn_HeadBand_l_04_jnt}
    constraintSetup [{targetBone {boneName=clavicle_l}}, {targetBone {boneName=spine_05}}]
    constraintWeights [1,1]
   }
   n13 = "Constraint - dyn_HeadBand_r_04_jnt (PP)"!(ComponentPose=@n12, Alpha=1, ConstraintWeights_0=0.75, ConstraintWeights_1:float=0.25) -> (Pose) {
    alphaBoolBlend {blendOption=Linear}
    boneToModify {boneName=dyn_HeadBand_r_04_jnt}
    constraintSetup [{targetBone {boneName=clavicle_r}}, {targetBone {boneName=spine_05}}]
    constraintWeights [1,1]
   }
   n14 = "Transform (Modify) Bone - Bone: dyn_HeadBand_r_04_jnt"(ComponentPose=@n13, Alpha=1, Translation="0,-1,0", Rotation="0,0,0", Scale="1,1,1") {
    rotationMode=BMM_Additive
    rotationSpace=BCS_BoneSpace
    translationMode=BMM_Additive
    translationSpace=BCS_BoneSpace
    alphaBoolBlend {blendOption=Linear}
    boneToModify {boneName=dyn_HeadBand_r_04_jnt}
    translation {y=-1}
   }
   n15 = RigidBody!(
    ComponentPose=@n14
    Alpha
    OverrideWorldGravity:Vector=Get Gravity Override()
    ExternalForce
    ComponentLinearAccScale:Vector="0.2,0.2,0.2"
    ComponentLinearVelScale:Vector="0.4,0.3,0.3"
   ) -> (Pose) {
    bClampLinearTranslationLimitToRefPose=true
    bOverrideWorldGravity=true
    overridePhysicsAsset=&/M_MED_JoyfulGrin_FaceAcc_Skydive_Physics
    simulationSpace=BaseBoneSpace
    alphaBoolBlend {blendOption=Linear}
    alphaScaleBiasClamp {bInterpResult=true, interpSpeedIncreasing=2}
    baseBoneRef {boneName=neck_02}
    componentAppliedLinearAccClamp=<3000 3000 3000>
    componentLinearAccScale=<0.4 0.3 0.3>
    componentLinearVelScale=<0.4 0.3 0.3>
    simSpaceSettings {
     maxAngularAcceleration=1000
     maxAngularVelocity=3
     maxLinearAcceleration=700
     maxLinearVelocity=500
     velocityScaleZ=0
     worldAlpha=0.25
     externalLinearDragV=<0.2 0.2>
    }
   }
   n16 = Control Rig(Source=Copy Pose From Mesh() {bUseAttachedParent=true}) {
    alphaBoolBlend {blendOption=Linear}
    controlRigAssetReference {blueprintRigClass=&M_MED_JoyfulGrin_FaceAcc_CtrlRig_C}
   }
   n17 = "Constraint - dyn_HeadBand_l_01_jnt (PP)"(ComponentPose=Local To Component(LocalPose=@n16), Alpha=1, ConstraintWeights_0=0.9, ConstraintWeights_1=0.1) {
    alphaBoolBlend {blendOption=Linear}
    boneToModify {boneName=dyn_HeadBand_l_01_jnt}
    constraintSetup [{targetBone {boneName=head}}, {targetBone {boneName=neck_01}}]
    constraintWeights [1,1]
   }
   n18 = "Constraint - dyn_HeadBand_r_01_jnt (PP)"(ComponentPose=@n17, Alpha=1, ConstraintWeights_0=0.9, ConstraintWeights_1=0.1) {
    alphaBoolBlend {blendOption=Linear}
    boneToModify {boneName=dyn_HeadBand_r_01_jnt}
    constraintSetup [{targetBone {boneName=head}}, {targetBone {boneName=neck_01}}]
    constraintWeights [1,1]
   }
   n19 = "Constraint - dyn_HeadBand_l_02_jnt (PP)"(ComponentPose=@n18, Alpha=1, ConstraintWeights_0=0.1, ConstraintWeights_1=0.9) {
    alphaBoolBlend {blendOption=Linear}
    boneToModify {boneName=dyn_HeadBand_l_02_jnt}
    constraintSetup [{targetBone {boneName=clavicle_l}}, {targetBone {boneName=neck_01}}]
    constraintWeights [1,1]
   }
   n20 = "Constraint - dyn_HeadBand_r_02_jnt (PP)"(ComponentPose=@n19, Alpha=1, ConstraintWeights_0=0.1, ConstraintWeights_1=0.9) {
    alphaBoolBlend {blendOption=Linear}
    boneToModify {boneName=dyn_HeadBand_r_02_jnt}
    constraintSetup [{targetBone {boneName=clavicle_r}}, {targetBone {boneName=neck_01}}]
    constraintWeights [1,1]
   }
   n21 = "Transform (Modify) Bone - Bone: dyn_HeadBand_l_02_jnt"(ComponentPose=@n20, Alpha=1, Translation="0,-2,0", Rotation="0,0,0", Scale="1,1,1") {
    rotationMode=BMM_Additive
    rotationSpace=BCS_BoneSpace
    translationMode=BMM_Additive
    translationSpace=BCS_BoneSpace
    alphaBoolBlend {blendOption=Linear}
    boneToModify {boneName=dyn_HeadBand_l_02_jnt}
    translation {y=-2}
   }
   n22 = "Transform (Modify) Bone - Bone: dyn_HeadBand_r_02_jnt"(ComponentPose=@n21, Alpha=1, Translation="0,-1,0", Rotation="0,0,60", Scale="1,1,1") {
    rotationMode=BMM_Additive
    rotationSpace=BCS_BoneSpace
    translationMode=BMM_Additive
    translationSpace=BCS_BoneSpace
    alphaBoolBlend {blendOption=Linear}
    boneToModify {boneName=dyn_HeadBand_r_02_jnt}
    rotation {roll=60}
    translation {y=-1}
   }
   n23 = "Constraint - dyn_HeadBand_l_03_jnt (PP)"(ComponentPose=@n22, Alpha=1, ConstraintWeights_0=0.5, ConstraintWeights_1=0.5) {
    alphaBoolBlend {blendOption=Linear}
    boneToModify {boneName=dyn_HeadBand_l_03_jnt}
    constraintSetup [{targetBone {boneName=clavicle_l}}, {targetBone {boneName=spine_05}}]
    constraintWeights [1,1]
   }
   n24 = "Constraint - dyn_HeadBand_r_03_jnt (PP)"(ComponentPose=@n23, Alpha=1, ConstraintWeights_0=0.5, ConstraintWeights_1=0.5) {
    alphaBoolBlend {blendOption=Linear}
    boneToModify {boneName=dyn_HeadBand_r_03_jnt}
    constraintSetup [{targetBone {boneName=clavicle_r}}, {targetBone {boneName=spine_05}}]
    constraintWeights [1,1]
   }
   n25 = "Transform (Modify) Bone - Bone: dyn_HeadBand_l_03_jnt"(ComponentPose=@n24, Alpha=1, Translation="0,2,0", Rotation="0,0,0", Scale="1,1,1") {
    rotationMode=BMM_Additive
    rotationSpace=BCS_BoneSpace
    translationMode=BMM_Additive
    translationSpace=BCS_BoneSpace
    alphaBoolBlend {blendOption=Linear}
    boneToModify {boneName=dyn_HeadBand_l_03_jnt}
    translation {y=2}
   }
   n26 = "Transform (Modify) Bone - Bone: dyn_HeadBand_r_03_jnt"(ComponentPose=@n25, Alpha=1, Translation="0,0,0", Rotation="0,0,0", Scale="1,1,1") {
    rotationMode=BMM_Additive
    rotationSpace=BCS_BoneSpace
    translationMode=BMM_Additive
    translationSpace=BCS_BoneSpace
    alphaBoolBlend {blendOption=Linear}
    boneToModify {boneName=dyn_HeadBand_r_03_jnt}
   }
   n27 = "Constraint - dyn_HeadBand_l_04_jnt (PP)"(ComponentPose=@n26, Alpha=1, ConstraintWeights_0=0.75, ConstraintWeights_1=0.25) {
    alphaBoolBlend {blendOption=Linear}
    boneToModify {boneName=dyn_HeadBand_l_04_jnt}
    constraintSetup [{targetBone {boneName=clavicle_l}}, {targetBone {boneName=spine_05}}]
    constraintWeights [1,1]
   }
   n28 = "Constraint - dyn_HeadBand_r_04_jnt (PP)"(ComponentPose=@n27, Alpha=1, ConstraintWeights_0=0.75, ConstraintWeights_1=0.25) {
    alphaBoolBlend {blendOption=Linear}
    boneToModify {boneName=dyn_HeadBand_r_04_jnt}
    constraintSetup [{targetBone {boneName=clavicle_r}}, {targetBone {boneName=spine_05}}]
    constraintWeights [1,1]
   }
   n29 = "Transform (Modify) Bone - Bone: dyn_HeadBand_r_04_jnt"(ComponentPose=@n28, Alpha=1, Translation="0,0,0", Rotation="0,0,0", Scale="1,1,1") {
    rotationMode=BMM_Additive
    rotationSpace=BCS_BoneSpace
    translationMode=BMM_Additive
    translationSpace=BCS_BoneSpace
    alphaBoolBlend {blendOption=Linear}
    boneToModify {boneName=dyn_HeadBand_r_04_jnt}
   }
   n30 = RigidBody(
    ComponentPose=@n29
    OverrideWorldGravity=Get Gravity Override()
    ComponentLinearAccScale="0.2,0.2,0.2"
    ComponentLinearVelScale="0.4,0.3,0.3"
   ) {
    bClampLinearTranslationLimitToRefPose=true
    bOverrideWorldGravity=true
    overridePhysicsAsset=&/M_MED_JoyfulGrin_FaceAcc_Ground_Physics
    simulationSpace=BaseBoneSpace
    alphaBoolBlend {blendOption=Linear}
    alphaScaleBiasClamp {bInterpResult=true, interpSpeedIncreasing=2}
    baseBoneRef {boneName=neck_02}
    componentAppliedLinearAccClamp=<3000 3000 3000>
    componentLinearAccScale=<0.4 0.3 0.3>
    componentLinearVelScale=<0.4 0.3 0.3>
    simSpaceSettings {
     maxAngularAcceleration=1000
     maxAngularVelocity=3
     maxLinearAcceleration=700
     maxLinearVelocity=500
     velocityScaleZ=0
     worldAlpha=0.25
     externalLinearDragV=<0.2 0.2>
    }
   }
   Output Pose(Result=Blend Poses by bool(
     BlendPose_0=Component To Local(ComponentPose=@n15)
     BlendPose_1=Component To Local(ComponentPose=@n30)
     bActiveValue=Get Is Skydive Bool Alpha() -> (Is Skydive Bool Alpha:bool)
    ) {blendType=Linear}) {layerGroup=None}
   Get bIsSkyDiving()
  ]}
 EventGraph {nodes [
   Event BlueprintUpdateAnimation()
   > n1 = Sequence()
   > n2 = Gravity Override State Function DEPRECATED(CharacterPartAnimInstance=Self-Reference(), GameTimeInSeconds=GetGameTimeInSeconds(), GravityOverrideParams=Get Gravity Override Params())
   > Set Gravity Override(Gravity Override=@n2)
   n3 = OR Boolean(A=Get bIsSkydiveFloating(), B=Get bIsSkydiveDiveMode())
   Set Is Skydiving Alpha(execute=@n1.then_1, Is Skydiving Alpha:double="To Float (Boolean)"(InBool=@n3)) -> (Output_Get:double)
   > Set Is Skydive Bool Alpha(Is Skydive Bool Alpha:bool=@n3) -> (Output_Get:bool)
   TryGetPawnOwner()
   Get bIsSkyDiving()
  ]}
]
variables [
 Gravity Override {type=Vector}
 Gravity Override Params {type=GravityOverrideParamsStruct}
 Is Skydive Bool Alpha {type=boolean}
 Is Skydiving Alpha {type=number}
]
